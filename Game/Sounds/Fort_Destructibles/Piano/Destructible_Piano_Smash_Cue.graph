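# Sound cue for the smash of a destructible piano prop, assigned to the Fort_Impacts sound class.
# A single wave player feeds a Modulator left at default settings, which slightly varies pitch and volume on each play.
# Overridden concurrency limits simultaneous instances and applies volume scaling as they stack.
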SoundCue /Game/Sounds/Fort_Destructibles/Piano/Destructible_Piano_Smash_Cue
root=Modulator_0
nodes [
 {id=Modulator_0, children [WavePlayer_0]}
 {id=WavePlayer_0, properties {soundWaveAssetPtr=&/Prop_Destructible_Piano_Smash_01}}
]
properties {
 bOverrideConcurrency=true
 maxDistance=2097152
 soundClassObject=&Fort_Impacts
 volumeMultiplier=1
 concurrencyOverrides {maxCount=6, volumeScale=0.9048}
}
